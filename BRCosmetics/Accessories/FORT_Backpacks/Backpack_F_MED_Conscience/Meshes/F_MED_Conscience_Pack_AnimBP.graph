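# Animation Blueprint for the Conscience backpack, giving it a floating, springy body and an animated eye.
# AnimDynamics springs the pack bone, with alpha and velocity/acceleration scales retuned for crouch, sprint and jump/fall, then layered sine waves add hover.
# The eye lerps through a looping rotator array with random pauses, while the pupil pulses with scale tied to gaze angle.
# Eye and pupil offsets blend out during fast movement, swimming or vehicles, except ziplining.
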AnimBlueprint /BRCosmetics/Accessories/FORT_Backpacks/Backpack_F_MED_Conscience/Meshes/F_MED_Conscience_Pack_AnimBP
target_skeleton=&/F_MED_Conscience_Pack_Skeleton
graphs [
 AnimGraph {nodes [
   n1 = Anim Dynamics - Bone: pack!(
    ComponentPose=Local To Component(LocalPose=Input Pose())
    Alpha=Get Alpha_anim_dynamic() -> (Alpha_anim_dynamic:double)
    GravityOverride:Vector="0,0,20"
    ComponentLinearAccScale:Vector=Get Linear_Acc_Scale() -> (Linear_Acc_Scale:Vector)
    ComponentLinearVelScale:Vector=Get Linear_Vel_Scale() -> (Linear_Vel_Scale:Vector)
    ExternalForce="0,0,0"
   ) -> (Pose) {
    alpha=0
    angularBiasOverride=1
    angularDampingOverride=0.75
    angularSpringConstant=20
    bGravityOverrideInSimSpace=true
    bLinearSpring=true
    bOverrideLinearDamping=true
    bUseGravityOverride=true
    bUsePlanarLimit=false
    bUseSphericalLimits=true
    gravityScale=0
    linearDampingOverride=0.81
    linearSpringConstant=22
    simulationSpace=BoneRelative
    alphaBoolBlend {blendOption=Linear}
    alphaScaleBiasClamp {bInterpResult=true}
    boundBone {boneName=pack}
    componentAppliedLinearAccClamp=<65 55 50>
    componentLinearAccScale=<0.7 0.7 1.5>
    componentLinearVelScale=<0.7 0.7 1.5>
    gravityOverride {z=20}
    physicsBodyDefinitions [{
      sphereCollisionRadius=0
      boundBone {boneName=pack}
      boxExtents=<20 8 20>
      constraintSetup {
       linearXLimitType=Free
       linearZLimitType=Free
       angularLimitsMax=<0 0 0>
       angularLimitsMin=<0 0 0>
       angularTarget=<0 0 0>
       linearAxesMax=<2.5 0.5 0>
       linearAxesMin=<-2.5 -10 0>
      }
      localJointOffset=<0 0 0>
     }]
    relativeSpaceBone {boneName=root}
    sphericalLimits [{limitRadius=7, limitType=Inner, drivingBone {boneName=pack}, sphereLocalOffset=<0 0 4.4>}]
   }
   n2 = Get is_Idle() -> (is_Idle:bool)
   n3 = "Transform (Modify) Bone - Bone: dyn_eye"!(ComponentPose=@n1, bAlphaBoolEnabled:bool=@n2, Translation="0,0,0", Rotation=Get Eye_rotation() -> (Eye_rotation:Rotator), Scale="1,1,1") -> (Pose) {
    alphaInputType=Bool
    rotationMode=BMM_Additive
    rotationSpace=BCS_BoneSpace
    alphaBoolBlend {blendInTime=0.8, blendOption=Linear, blendOutTime=0.8}
    boneToModify {boneName=dyn_eye}
   }
   n4 = "Transform (Modify) Bone - Bone: dyn_pupil"!(ComponentPose=@n3, bAlphaBoolEnabled:bool=@n2, Translation="0,0,0", Rotation="0,0,0", Scale=Get Pupil_scale() -> (Pupil_scale:Vector)) -> (Pose) {
    alphaInputType=Bool
    scaleMode=BMM_Additive
    scaleSpace=BCS_BoneSpace
    alphaBoolBlend {blendInTime=0.8, blendOption=Linear, blendOutTime=0.8}
    boneToModify {boneName=dyn_pupil}
    scale=<0 0 0>
   }
   n5 = "Transform (Modify) Bone - Bone: pack"(ComponentPose=@n4, Translation=Get Backpack_floating() -> (Backpack_floating:Vector), Rotation="0,0,0") {translationMode=BMM_Additive, translationSpace=BCS_BoneSpace, alphaBoolBlend {blendOption=Linear}, boneToModify {boneName=pack}}
   Output Pose(Result=Component To Local(ComponentPose=@n5)) {layerGroup=None}
  ]}
 EventGraph {nodes [
   Pupil_and_condition() {is=title K2Node_CustomEvent}
   n1 = float * float(A=GetGameTimeInSeconds(), B=2) {is=title K2Node_PromotableOperator}
   n2 = float + float(A=float * float(A="Sin (Radians)"(A=@n1), B=0.3), B=1.5)
   n3 = Get Delta_time_X() -> (Delta_time_X:double)
   n4 = Get Pupil_scale_interp() -> (Pupil_scale_interp:double)
   n5 = BreakRotator(InRot=Get Eye_position_2() -> (Eye_position_2:Rotator))
   n6 = Vector Length(A=MakeVector(X=@n5, Y=@n5.Pitch, Z=@n5.Yaw))
   n7 = BreakRotator(InRot=Get Eye_position_1() -> (Eye_position_1:Rotator))
   n8 = Vector Length(A=MakeVector(X=@n7, Y=@n7.Pitch, Z=@n7.Yaw))
   n9 = Lerp(A=@n8, B=@n6, Alpha=Get alpha_position() -> (alpha_position:double))
   > Set Pupil_scale_interp(Pupil_scale_interp:double=float * float(
     A=MapRangeClamped(Value=@n9, InRangeA=5, InRangeB=27, OutRangeA=0.8, OutRangeB=1.1)
     B=FInterpTo(Current=@n4, Target=@n2, DeltaTime=@n3, InterpSpeed=40)
    )) -> (Output_Get:double)
   n10 = Get Pupil_scale_interp()
   > n11 = Set Pupil_scale(Pupil_scale:Vector=MakeVector(X=@n10, Y=1, Z=@n10)) -> (Output_Get:Vector)
   n12 = Cast To PlayerPawn_Athena(execute=@n39.then, Object=TryGetPawnOwner())
   n13 = IsZiplining!(execute=@n11.then, self:Fort Player Pawn Object Reference=@n12) -> (then, ReturnValue:bool)
   n14 = Get PawnVelocity()
   n15 = OR Boolean!(A="float > float"(A=Vector Length(A=@n14), B=551), B=Get bIsSurfaceSwimming(), C:bool=Get bIsInVehicle() -> (bIsInVehicle:bool)) -> (ReturnValue)
   > Set is_Idle(is_Idle:bool=OR Boolean!(A=@n13, B=NOT Boolean(A=@n15)) -> (ReturnValue)) -> (Output_Get:bool)
   n16 = Event BlueprintUpdateAnimation()
   > n17 = Set Delta_time_X(Delta_time_X:double=@n16.DeltaTimeX) -> (Output_Get:double)
   n18 = Get array_index() -> (array_index:int=0)
   n19 = Get eye_rotation_array() -> (eye_rotation_array:Array of Rotators)
   > Set Eye_position_1(Eye_position_1:Rotator="Get (a copy)"(Array:Array of Rotators=@n19, Dimension 1=@n18) -> (Output:Rotator)) -> (Output_Get:Rotator)
   > Set Eye_position_2(Eye_position_2:Rotator="Get (a copy)"(Array=@n19, Dimension 1=int + int(A=@n18))) -> (Output_Get:Rotator)
   > n20 = Sequence!(execute) -> (then_0, then_1, then_2:Exec)
   n21 = float * float!(A=Get period() -> (period:double), B=Get Rotation_speed_coefficient() -> (Rotation_speed_coefficient:double), C:double=1.5) -> (ReturnValue)
   > n22 = Set alpha_position(alpha_position:double=float + float(A=float / float(A=@n17, B=@n21), B=Get alpha_position())) -> (Output_Get:double)
   n23 = float - float(A=@n8, B=@n6)
   > Set Rotation_speed_coefficient(Rotation_speed_coefficient:double=MapRangeClamped(Value="Absolute (Float)"(A=@n23), InRangeA=0, InRangeB=27, OutRangeA=0.5)) -> (Output_Get:double)
   > n24 = Branch(Condition="float < float"(A=@n22, B=1))
   > Set Eye_rotation(Eye_rotation:Rotator="Lerp (Rotator)"!(A:Rotator=Get Eye_position_1(), B:Rotator=Get Eye_position_2(), Alpha:float=Get alpha_position(), bShortestPath:bool=false) -> (ReturnValue:Rotator)) -> (Output_Get:Rotator)
   Delay(execute=@n24.else, Duration=RandomFloatInRange(Min=0.3, Max=0.6))
   > Set alpha_position(alpha_position=0)
   > Increment Int(Value=Get array_index() -> (array_index=0))
   n25 = Length(TargetArray:Array of Rotators=Get eye_rotation_array())
   > Branch(Condition="integer >= integer"(A=Get array_index() -> (array_index=0), B=int - int(A=@n25)))
   > Set array_index(array_index:int=0) -> (Output_Get:int=0)
   Pupil_and_condition(execute=@n20.then_1) {is=title K2Node_CallFunction}
   n26 = Sequence!(execute=@n20.then_2) -> (then_0, then_1)
   n27 = float * float!(A=GetGameTimeInSeconds(), B=1.2) -> (ReturnValue)
   n28 = "Sin (Radians)"(A=float * float(A=1.21, B=@n27) {is=title K2Node_CommutativeAssociativeBinaryOperator})
   n29 = "Sin (Radians)"(A=float * float(A=0.56, B=@n27))
   n30 = "Sin (Radians)"(A=float * float(A=0.8, B=@n27))
   > Set Backpack_floating(Backpack_floating:Vector=MakeVector(X=float * float(A=@n29, B=0.5), Y=float * float(A=@n28, B=0.15), Z=float * float(A=@n30, B=1))) -> (Output_Get:Vector)
   n31 = Branch(execute=@n26.then_1, Condition=Get bIsCrouching())
   > Set Alpha_anim_dynamic(Alpha_anim_dynamic:double=0.3) -> (Output_Get:double)
   > Set Linear_Acc_Scale(Linear_Acc_Scale:Vector="0,0,0.1") -> (Output_Get:Vector)
   > n32 = Set Linear_Vel_Scale(Linear_Vel_Scale:Vector="1.2,1.2,0.6") -> (Output_Get:Vector)
   n33 = Get bIsCrouching()
   n34 = Branch(execute=[@n38.then, @n32.then], Condition=AND Boolean(A=Get bIsSprinting(), B=NOT Boolean(A=@n33)))
   > Set Alpha_anim_dynamic(Alpha_anim_dynamic=0.1)
   > Set Linear_Acc_Scale(Linear_Acc_Scale="0,0,0.1")
   > n35 = Set Linear_Vel_Scale(Linear_Vel_Scale="1,1,1.2")
   n36 = Branch(execute=[@n37.then, @n35.then], Condition=OR Boolean(A=Get bIsJumping(), B=Get bIsFalling()))
   > Set Alpha_anim_dynamic(Alpha_anim_dynamic=0.1)
   > Set Linear_Acc_Scale(Linear_Acc_Scale="0, 0, 0")
   > Set Linear_Vel_Scale(Linear_Vel_Scale="0.2,0.2,0.8")
   Set Alpha_anim_dynamic(execute=@n36.else, Alpha_anim_dynamic=0.7)
   > Set Linear_Acc_Scale(Linear_Acc_Scale="1.1,1.1,1.3")
   > Set Linear_Vel_Scale(Linear_Vel_Scale="1.1,1.1,1.3")
   Set Alpha_anim_dynamic(execute=@n34.else, Alpha_anim_dynamic=0.7)
   > Set Linear_Vel_Scale(Linear_Vel_Scale="1.1,1.1,1.3")
   > n37 = Set Linear_Vel_Scale(Linear_Vel_Scale="1.1,1.1,1.3")
   Set Alpha_anim_dynamic(execute=@n31.else, Alpha_anim_dynamic=0.7)
   > Set Linear_Vel_Scale(Linear_Vel_Scale="1.1,1.1,1.3")
   > n38 = Set Linear_Vel_Scale(Linear_Vel_Scale="1.1,1.1,1.3")
   n39 = Event BlueprintInitializeAnimation()
  ]}
]
variables [
 Alpha_anim_dynamic {type=number}
 alpha_position {type=number}
 array_index {type=integer}
 Backpack_floating {type=Vector}
 Delta_time_X {type=number}
 Eye_position_1 {type=Rotator}
 Eye_position_2 {type=Rotator}
 Eye_rotation {type=Rotator}
 eye_rotation_array {
  type="array<Rotator>"
  default [
   <rot 0 0 0>
   <rot 0 25 12>
   <rot 0 10 8>
   <rot 0 0 0>
   <rot 0 -20 -10>
   <rot 0 -18 10>
   <rot 0 -8 15>
   <rot 0 20 7>
   <rot 0 15 -15>
   <rot 0 0 0>
  ]
 }
 is_Idle {type=boolean}
 Linear_Acc_Scale {type=Vector}
 Linear_Vel_Scale {type=Vector}
 period {type=number, default=0.1}
 Pupil_interp_speed {type=number}
 Pupil_scale {type=Vector}
 Pupil_scale_interp {type=number}
 Rotation_speed_coefficient {type=number}
]
